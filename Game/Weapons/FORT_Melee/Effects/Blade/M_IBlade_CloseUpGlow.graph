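# Unlit translucent material for Niagara mesh particles, apparently a close-up glow on the blade melee weapon.
# Particle color RGB runs through an emissive adjustment function, while particle alpha times Fresnel forms a rim mask.
# On the default shading path, a powered, low-clamped Near_Camera_Fade scales that clamped mask; mobile skips this fade.
# A distance synced fade function is also present, sharing the fade start parameter name.
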Material /Game/Weapons/FORT_Melee/Effects/Blade/M_IBlade_CloseUpGlow
graph {
 properties {
  bUsedWithNiagaraMeshParticles=true
  blendMode=BLEND_Translucent
  pixelDepthOffsetMode=PDOM_Legacy
  shadingModel=MSM_Unlit
  usedShadingModels=Unlit
 }
 nodes [
  n1 = ParticleColor()
  n2 = Multiply(A=@n1.A, B=Fresnel())
  n3 = MaterialFunctionCall(Distance=ScalarParameter() {defaultValue=1024, parameterName="Fade_Start_Distance (S)"}) {materialFunction=&Near_Camera_Fade}
  n4 = Multiply(A=Clamp(None=Power(Base=@n3.Result) {constExponent=4}) {maxDefault=0.15})
  Clamp(None=ShadingPathSwitch(Default=Multiply(A=@n2, B=@n4), Mobile=@n2) {desc=Migrated from FeatureLevelSwitch})
  MaterialFunctionCall(
   Fade_Start_Distance=ScalarParameter() {defaultValue=1024, parameterName="Fade_Start_Distance (S)"}
   Fade_End_Distance=ScalarParameter() {defaultValue=1024, parameterName="Fade_End_Distance (S)"}
  ) {materialFunction=&MF_Distance_SyncedFade_FX}
  MaterialFunctionCall(Emissive=@n1.RGB) {materialFunction=&MF_FX_EmissiveAdjustments}
 ]
}
