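# Translucent lit material for Niagara mesh particles that draws animated streak effects on a melee blade.
# PanRed, PanGreen and PanBlue dynamic parameters each scroll a noise-distorted vertical gradient, replacing time in the panners.
# Each gradient becomes a sharp line over a faint background, masked by its own IB_Mask channel.
# The three layers are summed and scaled by particle alpha, while particle color feeds a shared emissive-adjustment function.
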Material /Game/Weapons/FORT_Melee/Effects/Blade/M_IBlade_EmissiveFX
graph {
 properties {bUsedWithNiagaraMeshParticles=true, blendMode=BLEND_Translucent, pixelDepthOffsetMode=PDOM_Legacy, usedShadingModels=Default Lit}
 nodes [
  Constant3Vector()
  MaterialFunctionCall(Speed=Constant() {r=0.2}) {materialFunction=&DebugTimeSine}
  n1 = ParticleColor()
  MaterialFunctionCall(Emissive=@n1.RGB) {materialFunction=&MF_FX_EmissiveAdjustments}
  MaterialFunctionCall(Speed=Constant() {r=0.2}) {materialFunction=&DebugTimeSine}
  MaterialFunctionCall(Speed=Constant() {r=0.2}) {materialFunction=&DebugTimeSine}
  n2 = TextureSample() {samplerType=SAMPLERTYPE_Masks, texture=/Game/Weapons/FORT_Melee/Blade_Athena/Textures/IB_Mask}
  n3 = Multiply(A=Add(A=DynamicParameter() {defaultValue=<rgba 0.5 0.5 0.5 1>, paramNames [PanRed,PanGreen,PanBlue,Param4]}.PanRed) {constB=-3.2}) {constB=5}
  n4 = TextureCoordinate()
  n5 = ScalarParameter() {defaultValue=2, parameterName=Distort_Texture_Tiling}
  n6 = TextureSample(UVs=Panner(Coordinate=Multiply(A=@n4, B=@n5), Time=@n3) {speedY=1}) {texture=&T_NoisePack_01}
  n7 = Add(
   A=Panner(Coordinate=TextureCoordinate(), Time=@n3) {speedY=0.25}
   B=Multiply(A=ComponentMask(None=@n6.RGB) {g=true, r=true}, B=ScalarParameter() {defaultValue=0.1, parameterName=Dist_Mult})
  )
  n8 = Constant2Vector() {g=3}
  n9 = ScalarParameter() {defaultValue=-1, desc=-1 or 1 scale, parameterName=FlipGradientVertically, sliderMax=1, sliderMin=-1}
  n10 = TextureSampleParameter2D(UVs=Multiply(A=Add(A=@n7, B=@n8), B=@n9) {constB=-1}) {parameterName=TexRed, texture=/Game/Global/Textures/Default/Masks_and_Gradients/T_Vert_Gradient_Soft}
  n11 = Multiply(A=Power(Base=@n10.RGB, Exp=ScalarParameter() {defaultValue=15, parameterName=Line_Power})) {constB=20}
  n12 = Multiply(A=@n10.RGB, B=ScalarParameter() {defaultValue=0.1, parameterName=Background_Texture_Intensity})
  n13 = Clamp(None=Multiply(B=Add(A=@n11, B=@n12)) {constA=4})
  n14 = Multiply(A=Add(A=DynamicParameter() {defaultValue=<rgba 0.5 0.5 0.5 1>, paramNames [PanRed,PanGreen,PanBlue,Param4]}.PanGreen) {constB=-3.2}) {constB=5}
  n15 = TextureCoordinate()
  n16 = ScalarParameter() {defaultValue=2, parameterName=Distort_Texture_Tiling}
  n17 = TextureSample(UVs=Panner(Coordinate=Multiply(A=@n15, B=@n16), Time=@n14) {speedY=1}) {texture=&T_NoisePack_01}
  n18 = Add(
   A=Panner(Coordinate=TextureCoordinate(), Time=@n14) {speedY=0.25}
   B=Multiply(A=ComponentMask(None=@n17.RGB) {g=true, r=true}, B=ScalarParameter() {defaultValue=0.1, parameterName=Dist_Mult})
  )
  n19 = Constant2Vector() {g=3}
  n20 = TextureSampleParameter2D(UVs=Multiply(A=Add(A=@n18, B=@n19), B=@n9) {constB=-1}) {parameterName=TexGreen, texture=&T_Vert_Gradient_Soft}
  n21 = ScalarParameter() {defaultValue=15, parameterName=Line_Power}
  n22 = Add(
   A=Multiply(A=Power(Base=@n20.RGB, Exp=@n21)) {constB=20}
   B=Multiply(A=@n20.RGB, B=ScalarParameter() {defaultValue=0.1, parameterName=Background_Texture_Intensity})
  )
  n23 = Multiply(
   A=TextureSample() {samplerType=SAMPLERTYPE_Masks, texture=&IB_Mask}.G
   B=Clamp(None=Multiply(B=@n22) {constA=4})
  )
  n24 = Multiply(A=Add(A=DynamicParameter() {defaultValue=<rgba 0.5 0.5 0.5 1>, paramNames [PanRed,PanGreen,PanBlue,Param4]}.PanBlue) {constB=-3.2}) {constB=5}
  n25 = TextureCoordinate()
  n26 = ScalarParameter() {defaultValue=2, parameterName=Distort_Texture_Tiling}
  n27 = TextureSample(UVs=Panner(Coordinate=Multiply(A=@n25, B=@n26), Time=@n24) {speedY=1}) {texture=&T_NoisePack_01}
  n28 = Add(
   A=Panner(Coordinate=TextureCoordinate(), Time=@n24) {speedY=0.25}
   B=Multiply(A=ComponentMask(None=@n27.RGB) {g=true, r=true}, B=ScalarParameter() {defaultValue=0.1, parameterName=Dist_Mult})
  )
  n29 = Constant2Vector() {g=3}
  n30 = TextureSampleParameter2D(UVs=Multiply(A=Add(A=@n28, B=@n29), B=@n9) {constB=-1}) {parameterName=TexBlue, texture=&T_Vert_Gradient_Soft}
  n31 = ScalarParameter() {defaultValue=15, parameterName=Line_Power}
  n32 = Add(
   A=Multiply(A=Power(Base=@n30.RGB, Exp=@n31)) {constB=20}
   B=Multiply(A=@n30.RGB, B=ScalarParameter() {defaultValue=0.1, parameterName=Background_Texture_Intensity})
  )
  n33 = Multiply(
   A=TextureSample() {samplerType=SAMPLERTYPE_Masks, texture=&IB_Mask}.B
   B=Clamp(None=Multiply(B=@n32) {constA=4})
  )
  Multiply(A=@n1.A, B=Add(A=Multiply(A=@n2.R, B=@n13), B=Add(A=@n23, B=@n33))) {constA=1}
 ]
}
