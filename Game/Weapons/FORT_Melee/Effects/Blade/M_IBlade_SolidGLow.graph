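# M_IBlade_SolidGlow is a translucent, Default Lit material for a solid glow on melee blade effects, usable on Niagara mesh particles.
# Particle color RGB feeds the MF_FX_EmissiveAdjustments function to produce the glow color.
# Particle alpha is doubled, multiplied by a Fresnel term and clamped, giving an edge-weighted mask.
# Constant colors, a debug sine-time function and a Noise node are also present, with no visible link to that color or mask.
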Material /Game/Weapons/FORT_Melee/Effects/Blade/M_IBlade_SolidGLow.M_IBlade_SolidGlow
graph {
 properties {bUsedWithNiagaraMeshParticles=true, blendMode=BLEND_Translucent, pixelDepthOffsetMode=PDOM_Legacy, usedShadingModels=Default Lit}
 nodes [
  n1 = ParticleColor()
  Clamp(None=Multiply(A=Multiply(A=@n1.A) {constB=2}, B=Fresnel()))
  Constant3Vector()
  Constant3Vector() {constant=<rgba 0.3 1.2 3 0>}
  MaterialFunctionCall(Emissive=@n1.RGB) {materialFunction=&MF_FX_EmissiveAdjustments}
  MaterialFunctionCall() {materialFunction=&DebugTimeSine}
  Noise!(World Position, FilterWidth) {levelScale=1.5, outputMax=2, outputMin=0, scale=0.05}
 ]
}
