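# Niagara ribbon anim trail for the IBlade melee weapon, named for its jump move, with a bright HDR blue default color.
# A single emitter spawns trail particles between two skeletal mesh bones or sockets using shared FN_AnimTrail modules, with a ribbon ID offset utility.
# User parameters set color and a very short lifetime.
# Each update refreshes the trail from the mesh, dims RGB and fades alpha along a steeply falling curve.
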NiagaraSystem /Game/Weapons/FORT_Melee/Effects/Blade/NS_IBlade_AnimTrail_Jump
dependencies {
 usedDataInterfaces [&NiagaraDataInterfaceCurve]
 usedDynamicInputs []
 usedModules [
  &self:SystemSpawnScript.SystemScriptSource.SystemScriptGraph.NiagaraNodeAssignment_0.SetVariables_#1
  &SystemState
 ]
}
emitters {FN_AnimTrail002 {
  emitterSpawnScript [&FN_RibbonIDOffsetUtility()]
  emitterUpdateScript [
   &EmitterState(Life Cycle Mode=::NewEnumerator1, Loop Duration=1, UseLoopDelay=false)
   &FN_SpawnAnimTrail()
  ]
  particleSpawnScript [
   &FN_InitializeAnimTrailBoneIndecies(First Bone or Socket Index=1, Second Bone or Socket Index=0)
   &FN_InitializeAnimTrail(Color=$User.TrailColor, Lifetime=$User.LifeTime, SkeletalMesh=$System.Skeletal Mesh)
  ]
  particleUpdateScript [
   &SolveForcesAndVelocity(Clamp Velocity=false, Limit Acceleration=false)
   &FN_UpdateAnimTrail(Locked Particle RibbonWidth=1, SkeletalMesh=$System.Skeletal Mesh) {module=FN_UpdateAnimTrail001}
   &ScaleColor(
    Scale Alpha=&FloatFromCurve(FloatCurve=DataInterface() {propertyValues {Curve {keys [
         {arriveTangent=-0.8451, interpMode=RCIM_Cubic, leaveTangent=-0.8451, tangentMode=RCTM_User, time=-0.005, value=0.1806}
         {arriveTangent=-0.4947, interpMode=RCIM_Cubic, leaveTangent=-0.4947, time=0.135, value=0.0512}
         {arriveTangent=-0.0594, interpMode=RCIM_Cubic, leaveTangent=-0.0594, time=0.3451, value=0.0074}
         <at 0.9969 0 RCIM_Cubic>
        ]}}})
    Scale RGB=<0.4 0.4 0.4>
    ScaleA=true
    ScaleRGB=true
   )
  ]
  rendererClasses [&NiagaraRibbonRendererProperties]
  renderers [{rendererClass=&NiagaraRibbonRendererProperties}]
 }}
summary {
 systemName=NS_IBlade_AnimTrail_Jump
 emitters [{emitterName=FN_AnimTrail002, rendererClasses [&NiagaraRibbonRendererProperties]}]
 userVariables [
  User.LifeTime {type=NiagaraFloat, defaultValue=0.05, description=""}
  User.RibbonIDOffset {type=NiagaraInt32}
  User.SkeletalMesh {
   type=NiagaraDataInterfaceSkeletalMesh
   defaultValue {
    struct=NiagaraExt_VariableValue_DataInterface
    value {dataInterface=&self:NiagaraDataInterfaceSkeletalMesh_0}
   }
  }
  User.TrailColor {type=LinearColor, description="", defaultValue {struct=LinearColor, value=<rgba 12.7664 75.1776 200 1>}}
 ]
}
