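# Niagara system for a flare effect on a melee blade weapon, going by its name and folder.
# It holds four emitters, SwordFull plus SwordR, SwordG and SwordB, whose names suggest a full blade layer and separate red, green and blue layers.
# System State is the only listed module, with no renderers, data interfaces, dynamic inputs or user variables shown.
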NiagaraSystem /Game/Weapons/FORT_Melee/Effects/Blade/NS_IBlade_Flare01
dependencies {usedDataInterfaces [], usedDynamicInputs [], usedModules [&SystemState]}
emitters {
 SwordB {rendererClasses [], renderers []}
 SwordFull {rendererClasses [], renderers []}
 SwordG {rendererClasses [], renderers []}
 SwordR {rendererClasses [], renderers []}
}
summary {
 systemName=NS_IBlade_Flare01
 emitters [
  {emitterName=SwordFull, rendererClasses []}
  {emitterName=SwordR, rendererClasses []}
  {emitterName=SwordG, rendererClasses []}
  {emitterName=SwordB, rendererClasses []}
 ]
 userVariables []
}
