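# Niagara effect system for a flare on the IBlade melee weapon, stored with the melee blade effects.
# It is composed of four emitters named StreaksFine, RingOuter, Light and UpwardStreaks, suggesting fine and rising streaks, an outer ring and a light.
# SystemState is the only module listed, and the file shows no renderers, data interfaces, dynamic inputs or user variables.
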NiagaraSystem /Game/Weapons/FORT_Melee/Effects/Blade/NS_IBlade_Flare02
dependencies {usedDataInterfaces [], usedDynamicInputs [], usedModules [&SystemState]}
emitters {
 Light {rendererClasses [], renderers []}
 RingOuter {rendererClasses [], renderers []}
 StreaksFine {rendererClasses [], renderers []}
 UpwardStreaks {rendererClasses [], renderers []}
}
summary {
 systemName=NS_IBlade_Flare02
 emitters [
  {emitterName=StreaksFine, rendererClasses []}
  {emitterName=RingOuter, rendererClasses []}
  {emitterName=Light, rendererClasses []}
  {emitterName=UpwardStreaks, rendererClasses []}
 ]
 userVariables []
}
