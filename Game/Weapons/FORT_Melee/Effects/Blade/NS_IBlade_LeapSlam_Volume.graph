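# Niagara system for the IBlade melee weapon's leap slam, apparently the volume impact effect judging by its name.
# Eight emitters are named for its layers: debris, dust, sparks, streaks, a left or large shockwave (ShockwaveL), normal and wide flash meshes, and a light.
# It exposes no user variables; the only listed module is SystemState, and no renderers, data interfaces or dynamic inputs are recorded.
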NiagaraSystem /Game/Weapons/FORT_Melee/Effects/Blade/NS_IBlade_LeapSlam_Volume
dependencies {usedDataInterfaces [], usedDynamicInputs [], usedModules [&SystemState]}
emitters {
 Dust {rendererClasses [], renderers []}
 FlashMesh {rendererClasses [], renderers []}
 FlashMeshWide {rendererClasses [], renderers []}
 Light {rendererClasses [], renderers []}
 ShockwaveL {rendererClasses [], renderers []}
 Sparks {rendererClasses [], renderers []}
 Streaks {rendererClasses [], renderers []}
 debris {rendererClasses [], renderers []}
}
summary {
 systemName=NS_IBlade_LeapSlam_Volume
 emitters [
  {emitterName=debris, rendererClasses []}
  {emitterName=Sparks, rendererClasses []}
  {emitterName=Light, rendererClasses []}
  {emitterName=ShockwaveL, rendererClasses []}
  {emitterName=FlashMesh, rendererClasses []}
  {emitterName=Dust, rendererClasses []}
  {emitterName=FlashMeshWide, rendererClasses []}
  {emitterName=Streaks, rendererClasses []}
 ]
 userVariables []
}
